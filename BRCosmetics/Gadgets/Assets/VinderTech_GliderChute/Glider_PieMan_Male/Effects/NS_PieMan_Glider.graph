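# Niagara effects for the PieMan glider cosmetic: dangling drool, flies, smoke and mirrored mesh trails, all sampled from the glider's skeletal mesh.
# DroolPendulum hangs stretched mesh strands from vertices using pendulum constraints and emits death events, alongside a separate Drool emitter.
# Curl-noise flies are pulled toward a bone, and Fly_Trail reads their particles for ribbons.
# System exhaust, length and opacity drive the trails; front-end modifiers substitute preview motion.
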NiagaraSystem /BRCosmetics/Gadgets/Assets/VinderTech_GliderChute/Glider_PieMan_Male/Effects/NS_PieMan_Glider
dependencies {
 usedDataInterfaces [&NiagaraDataInterfaceCurve, &NiagaraDataInterfaceColorCurve, &NiagaraDataInterfaceParticleRead]
 usedDynamicInputs []
 usedModules [
  &SystemState
  &self:Set_Current_Exhaust
  &self:Set_Trail_Length
  &self:Set_Trail_Opacity
 ]
}
emitters {
 Drool {
  emitterUpdateScript [&EmitterState()]
  particleSpawnScript [/Niagara/Modules/Spawn/Initialization/InitializeParticle(
    Lifetime=&FN_FrontEndFloatModifier(
     Gameplay=&Lerp_Float(A=1, B=0, Alpha=&ClampFloat(Float=&Multiply_Float(A=2, B=$User.ForwardDot)))
     Front End=&RandomRangeFloat(Minimum=0.25, Maximum=0.5)
    )
    Mesh Scale=&Lerp_Vector(A=<10 3 3>, B=<25 6 6>, Alpha=$Particles.RandomNormalizedFloat)
    Module.Write Color=true
    Module.Write Lifetime=true
    Module.Write Mass=true
    Module.Write Material Random=true
    Module.Write Position=false
    Module.Write Scale=true
    Module.Write SpriteRotation=false
    Module.Write SpriteSize=true
    Sprite Size=&Vector2DFromFloat(Value=5)
   )]
  particleUpdateScript [
   &ParticleState()
   &DynamicMaterialParameters(
    Index 0 Param 1=&RandomRangeFloat(Minimum=0, Maximum=1)
    Index 0 Param 2=1
    Module.Param0WriteEnabled=true
    Module.Param1WriteEnabled=true
    Module.Param2WriteEnabled=true
    Module.Param3WriteEnabled=true
   )
   &ScaleMeshSize(Scale Factor=&Multiply_VectorByFloat(
     Vector=&MakeVector(
      X=&Lerp_Float(A=0.25, B=1, Alpha=$Particles.NormalizedAge)
      Y=&Lerp_Float(A=1, B=0.5, Alpha=$Particles.NormalizedAge)
      Z=&Lerp_Float(A=1, B=0.5, Alpha=$Particles.NormalizedAge)
     )
     Float=&Lerp_Float(A=1, B=0, Alpha=$Particles.NormalizedAge)
    ))
   &InheritVelocity(Module.Limit Speed=true, Velocity Scale=<-0.1 -0.1 -0.1>)
   &AddVelocity(Velocity=&FN_FrontEndVector3Modifier(
     Gameplay=<0 0 0>
     Front End=&RandomRangeVector(Minimum=<-165 80 0>, Maximum=<-100 50 50>, Evaluation Type=::NewEnumerator1)
    ))
   &AddVelocity(Velocity=&FN_FrontEndVector3Modifier(
     Gameplay=<0 0 0>
     Front End=&RandomRangeVector(Minimum=<-5 -300 -20>, Maximum=<5 -50 100>, Evaluation Type=::NewEnumerator1)
    )) {enabled=false, module=AddVelocity001}
   &SolveForcesAndVelocity(Clamp Velocity=false, Limit Acceleration=false)
  ]
  rendererClasses [&NiagaraSpriteRendererProperties, &NiagaraMeshRendererProperties]
  renderers [
   {rendererClass=&NiagaraSpriteRendererProperties}
   {rendererClass=&NiagaraMeshRendererProperties, rendererIndex=1}
  ]
 }
 DroolPendulum {
  emitterUpdateScript [
   &EmitterState()
   &SpawnRate(SpawnRate=10, Use Spawn Probability=false)
  ]
  particleSpawnScript [
   /SetVariables(
    Particles.PickAll:NiagaraInt32=&RandomRangeInt(Minimum=0, Maximum=1810)
    Particles.PickBack:NiagaraInt32=&RandomRangeInt(Minimum=1252, Maximum=1171)
   ) {isSetParameters=true}
   /SetVariables(
    Particles.PendulumLength:NiagaraFloat=&RandomRangeFloat(Minimum=4, Maximum=20)
    Particles.PickBone:NiagaraInt32=&MakeCustomIntFromBool(
     Bool=&SetBoolByIntComparison(A=&RandomRangeInt(Minimum=0, Maximum=1), B=0, Comparison Type=::NewEnumerator4)
     True INT=$Particles.PickBack
     False INT=$Particles.PickAll
    )
    Particles.ScaleLerp:NiagaraFloat=&RandomRangeFloat(Minimum=0, Maximum=1)
   ) {isSetParameters=true, module=SetVariables_#1}
   &SkeletalMeshLocation(
    Inherit Velocity=false
    MaskSpawnGroup=false
    Mesh Sampling Type=::NewEnumerator8
    OffsetAlongNormal=false
    Skeletal Mesh=$User.SkeletalMesh
    Vertex ID:NiagaraInt32=$Particles.PickBone
    Vertex Sampling Mode:ENiagaraVertexSamplingMode=::NewEnumerator8
   ) {module=SkeletalMeshLocation002}
   &InitializeParticle(
    Lifetime=&Power(a=&UniformRangedFloat(Minimum=0.5, Maximum=1.2), b=2)
    Mesh Scale=&Lerp_Vector(A=<0.25 0.25 0.02>, B=<0.75 0.75 0.02>, Alpha=$Particles.ScaleLerp)
    Module.Write Color=true
    Module.Write Lifetime=true
    Module.Write Mass=true
    Module.Write Material Random=true
    Module.Write Position=false
    Module.Write Scale=true
    Module.Write SpriteRotation=false
    Module.Write SpriteSize=true
    Sprite Size=&Vector2DFromFloat(Value=5)
   )
  ]
  particleUpdateScript [
   &ParticleState()
   &SkeletalMeshLocation(
    Apply To Particle Position=false
    Inherit Velocity=false
    MaskSpawnGroup=false
    Mesh Sampling Type=::NewEnumerator8
    OffsetAlongNormal=false
    Skeletal Mesh=$User.SkeletalMesh
    Vertex ID:NiagaraInt32=$Particles.PickBone
    Vertex Sampling Mode:ENiagaraVertexSamplingMode=::NewEnumerator8
   )
   /SetVariables(Particles.MeshPosition:Vector3f=$Output.SkeletalMeshLocation.SampledPosition) {isSetParameters=true, module=SetVariables_#2}
   /SetVariables(
    Particles.AlphaScale:NiagaraFloat=&FloatFromCurve(FloatCurve=DataInterface() {propertyValues {Curve {keys [
         {arriveTangent=-0.0275, interpMode=RCIM_Cubic, leaveTangent=-0.0275, tangentMode=RCTM_User, time=0.95, value=1}
         {arriveTangent=-34.2468, interpMode=RCIM_Cubic, leaveTangent=-34.2466, tangentMode=RCTM_User, time=1, value=0}
        ]}}})
    Particles.XYScale:NiagaraFloat=&FloatFromCurve(FloatCurve=DataInterface() {propertyValues {Curve {keys [
         {arriveTangent=10.1742, interpMode=RCIM_Cubic, leaveTangent=10.1742, tangentMode=RCTM_User, time=0, value=0}
         {arriveTangent=-2.2106, interpMode=RCIM_Cubic, leaveTangent=-2.2106, tangentMode=RCTM_User, time=0.1, value=1}
         {arriveTangent=-0.0096, interpMode=RCIM_Cubic, leaveTangent=-0.0096, tangentMode=RCTM_User, time=0.4, value=0.5}
         {arriveTangent=-3.238, interpMode=RCIM_Cubic, leaveTangent=-3.238, tangentMode=RCTM_User, time=1, value=0.1}
        ]}}})
    Particles.ZScale:NiagaraFloat=&FloatFromCurve(
     FloatCurve=DataInterface() {propertyValues {Curve {keys [
         {arriveTangent=4.2799, interpMode=RCIM_Cubic, leaveTangent=4.2799, tangentMode=RCTM_User, time=0, value=0.1}
         {arriveTangent=1.4091, interpMode=RCIM_Cubic, leaveTangent=1.4091, tangentMode=RCTM_User, time=0.5, value=1}
         {arriveTangent=10.2989, interpMode=RCIM_Cubic, leaveTangent=10.2989, tangentMode=RCTM_User, time=1, value=3}
        ]}}}
     Scale Curve=$Particles.PendulumLength
    )
   ) {isSetParameters=true}
   &PendulumSetup(
    Module.Enable Spring Driver=false
    Pendulum Length:NiagaraFloat=$Particles.ZScale
    Pendulum Pivot:Vector3f=$Particles.MeshPosition
    Rigid Pendulum=false
   )
   &Drag(Drag=2, Use Linear Drag=true, Use Rotational Drag=false)
   &AddVelocity(Velocity=&FN_FrontEndVector3Modifier(
     Gameplay=<0 0 0>
     Front End=&RandomRangeVector(Minimum=<-50 0 -20>, Maximum=<-200 100 100>, Evaluation Type=::NewEnumerator1)
    ))
   &AddVelocity(Velocity=&FN_FrontEndVector3Modifier(
     Gameplay=<0 0 0>
     Front End=&RandomRangeVector(Minimum=<-5 -300 -20>, Maximum=<5 -50 100>, Evaluation Type=::NewEnumerator1)
    )) {enabled=false, module=AddVelocity001}
   &SolveForcesAndVelocity(Clamp Velocity=false, Limit Acceleration=false)
   &PendulumConstraint()
   &CalculateAccurateVelocity()
   /SetVariables(Particles.MeshFacing:Vector3f=&NormalizeVector(Vector To Normalize=&Subtract_Vector(A=$Particles.MeshPosition, B=Linked() {linkedVariable {name=Particles.Position, type=NiagaraPosition}}))) {isSetParameters=true}
   /Niagara/Modules/Update/Orientation/V2/OrientMeshToVector(
    Look At Direction:Vector3f=&Multiply_VectorByFloat(Vector=$Particles.MeshFacing, Float=1)
    Override Quat=false
    Reference Axis Vector:Vector3f=<0 0 1>
    UseOptionalSideAxis=false
    UseOptionalUpAxis=false
   )
   &SkeletalMeshLocation(
    Apply To Particle Position=false
    Bone Sampling Mode=::NewEnumerator4
    Inherit Velocity=false
    MaskSpawnGroup=false
    Skeletal Mesh=$User.SkeletalMesh
   ) {module=SkeletalMeshLocation001}
   /SetVariables(Particles.RestDirection:Vector3f=/Niagara/DynamicInputs/Helpers/SafeNormalizeVector(Vector To Normalize:Vector3f=$Engine.Owner.Velocity, Fallback Vector:Vector3f=<0 0 0>)) {isSetParameters=true}
   &DynamicMaterialParameters(
    Index 1 Param 1=HlslExpression() {hlslExpression=Particles.RestDirection.x}
    Index 1 Param 2=HlslExpression() {hlslExpression=Particles.RestDirection.y}
    Index 1 Param 3=HlslExpression() {hlslExpression=Particles.RestDirection.z}
    Index 2 Param 1=HlslExpression() {hlslExpression=Particles.MeshPosition.x}
    Index 2 Param 2=HlslExpression() {hlslExpression=Particles.MeshPosition.y}
    Index 2 Param 3=HlslExpression() {hlslExpression=Particles.MeshPosition.z}
    Index 2 Param 4=0.75
    Index 3 Param 1=&MakeFloatFromVector(Vector=Linked() {linkedVariable {name=Particles.Position, type=NiagaraPosition}}, Channel=::NewEnumerator0)
    Index 3 Param 2=&MakeFloatFromVector(Vector=Linked() {linkedVariable {name=Particles.Position, type=NiagaraPosition}}, Channel=::NewEnumerator1)
    Index 3 Param 3=&MakeFloatFromVector(Vector=Linked() {linkedVariable {name=Particles.Position, type=NiagaraPosition}}, Channel=::NewEnumerator2)
    Index 3 Param 4=0.3
    Module.Param0WriteEnabled1=true
    Module.Param0WriteEnabled2=true
    Module.Param0WriteEnabled3=true
    Module.Param1WriteEnabled1=true
    Module.Param1WriteEnabled2=true
    Module.Param1WriteEnabled3=true
    Module.Param2WriteEnabled1=true
    Module.Param2WriteEnabled2=true
    Module.Param2WriteEnabled3=true
    Module.Param3WriteEnabled1=false
    Module.Param3WriteEnabled2=true
    Module.Param3WriteEnabled3=true
    Write Parameter Index 0=false
    Write Parameter Index 1=true
    Write Parameter Index 2=true
    Write Parameter Index 3=true
   )
   &ScaleMeshSize(Scale Factor=&MakeVector(
     X=$Particles.XYScale
     Y=$Particles.XYScale
     Z=&VectorLength(Vector=&Subtract_Vector(A=$Particles.MeshPosition, B=Linked() {linkedVariable {name=Particles.Position, type=NiagaraPosition}}))
    ))
   &DynamicMaterialParameters(
    Index 1 Param 1=HlslExpression() {hlslExpression="(Particles.MeshFacing.x+Particles.MeshNormal.x)/4"}
    Index 1 Param 2=HlslExpression() {hlslExpression="(Particles.MeshFacing.y+Particles.MeshNormal.y)/4"}
    Index 1 Param 3=HlslExpression() {hlslExpression="(Particles.MeshFacing.z+Particles.MeshNormal.z+2)/4"}
    Index 2 Param 1=HlslExpression() {hlslExpression=Particles.MeshPosition.x}
    Index 2 Param 2=HlslExpression() {hlslExpression=Particles.MeshPosition.y}
    Index 2 Param 3=HlslExpression() {hlslExpression=Particles.MeshPosition.z}
    Index 2 Param 4=1.5
    Index 3 Param 1=&MakeFloatFromVector(Vector=Linked() {linkedVariable {name=Particles.Position, type=NiagaraPosition}}, Channel=::NewEnumerator0)
    Index 3 Param 2=&MakeFloatFromVector(Vector=Linked() {linkedVariable {name=Particles.Position, type=NiagaraPosition}}, Channel=::NewEnumerator1)
    Index 3 Param 3=&MakeFloatFromVector(Vector=Linked() {linkedVariable {name=Particles.Position, type=NiagaraPosition}}, Channel=::NewEnumerator2)
    Index 3 Param 4=0.1
    Module.Param0WriteEnabled1=true
    Module.Param0WriteEnabled2=true
    Module.Param0WriteEnabled3=true
    Module.Param1WriteEnabled1=true
    Module.Param1WriteEnabled2=true
    Module.Param1WriteEnabled3=true
    Module.Param2WriteEnabled1=true
    Module.Param2WriteEnabled2=true
    Module.Param2WriteEnabled3=true
    Module.Param3WriteEnabled1=false
    Module.Param3WriteEnabled2=true
    Module.Param3WriteEnabled3=true
    Write Parameter Index 0=false
    Write Parameter Index 1=true
    Write Parameter Index 2=true
    Write Parameter Index 3=true
   ) {enabled=false, module=DynamicMaterialParameters002}
   &Color(Color=&ColorFromCurve(
     ColorCurve=DataInterface() {propertyValues {
       AlphaCurve {keys [<at 0 20 RCIM_Cubic>]}
       BlueCurve {keys [<at 0 0 RCIM_Cubic>,<at 0.5257 0>,<at 1 5 RCIM_Cubic>]}
       GreenCurve {keys [<at 0 0 RCIM_Cubic>,<at 0.5257 5>,<at 1 0 RCIM_Cubic>]}
       RedCurve {keys [<at 0 5 RCIM_Cubic>,<at 0.5257 0>,<at 1 0 RCIM_Cubic>]}
      }}
     CurveIndex=&Divide_Float(A=&MakeFloatFromInt(Int=$Particles.PickBone), B=366)
    )) {enabled=false}
   &ScaleColor(Scale Alpha=$Particles.AlphaScale, ScaleA=true, ScaleRGB=false)
   &GenerateDeathEvent(Float to Send as Random Float:NiagaraFloat=$Particles.ScaleLerp)
  ]
  rendererClasses [&NiagaraSpriteRendererProperties, &NiagaraMeshRendererProperties]
  renderers [
   {rendererClass=&NiagaraSpriteRendererProperties}
   {rendererClass=&NiagaraMeshRendererProperties, rendererIndex=1}
  ]
 }
 Flies_Optimized {
  emitterUpdateScript [
   &EmitterState()
   &SpawnRate(Spawn Probability=0.5, SpawnRate=5, Use Spawn Probability=true)
  ]
  particleSpawnScript [
   &InitializeParticle(
    Color=&RandomRangeLinearColor(Minimum=<rgba 0 0.0057 0.057 1>, Maximum=<rgba 0.005 0.02 0 1>)
    Lifetime=&Lerp_Float(A=1, B=3, Alpha=&Power(a=&RandomRangeFloat(Minimum=0, Maximum=1), b=4))
    Mesh Scale=&Multiply_VectorByFloat(Vector=<1 2 1>, Float=&UniformRangedFloat(Minimum=0.2, Maximum=0.4))
    Module.Write Color=true
    Module.Write Lifetime=true
    Module.Write Mass=false
    Module.Write Material Random=true
    Module.Write Position=false
    Module.Write Scale=true
    Module.Write SpriteRotation=false
    Module.Write SpriteSize=false
    Position=&SimulationPosition()
   )
   &DynamicMaterialParameters(
    Index 0 Param 1=10
    Index 0 Param 2=1
    Index 1 Param 1=&UniformRangedFloat(Minimum=0, Maximum=30)
    Index 1 Param 2=&UniformRangedFloat(Minimum=0, Maximum=1)
    Index 1 Param 3=&UniformRangedFloat(Minimum=0, Maximum=1)
    Index 1 Param 4=&UniformRangedFloat(Minimum=0, Maximum=1)
    Module.Param0WriteEnabled=true
    Module.Param0WriteEnabled1=true
    Module.Param1WriteEnabled=true
    Module.Param1WriteEnabled1=false
    Module.Param2WriteEnabled=true
    Module.Param2WriteEnabled1=false
    Module.Param3WriteEnabled=true
    Module.Param3WriteEnabled1=false
    Write Parameter Index 1=true
   )
   &SkeletalMeshLocation(
    Apply To Particle Position=false
    Bone / Socket Index=2
    Bone Sampling Mode=::NewEnumerator4
    Inherit Velocity=false
    MaskSpawnGroup=false
    Sampled Position Offset=&Add_Vector(
     A=<-22.1231 -11 -0.8408>
     B=&RotateVectorByQuaternion(Vector To Rotate=&RandomRangeVector(Minimum=<0 16.1011 0>, Maximum=<30 0 2>), Quaternion=<0 -0.15 0 1>)
    )
    Sampled Position Offset Coordinate Space:ENiagaraMeshTransforms=::NewEnumerator3
    Skeletal Mesh=$User.SkeletalMesh
   )
   /Niagara/Modules/Spawn/Location/SphereLocation(
    Hemisphere Z=true
    MaskSpawnGroup=false
    Module.Override Randomness=false
    Sphere Origin:Vector3f=$Output.SkeletalMeshLocation.SampledPosition
    Sphere Radius=20
    Surface Only=false
   )
   &InheritVelocity(Module.Limit Speed=false, Velocity Scale=<0.2 0.2 0.75>)
   &AddVelocity(Velocity=&Multiply_VectorByFloat(
     Vector=&RandomRangeVector(Minimum=<200 200 0>, Maximum=<500 500 0>)
     Float=&UniformAOrBFloat(A=-1)
    ))
   &AddVelocity(Velocity=&FN_FrontEndVector3Modifier(Gameplay=<0 0 0>, Front End=&RandomRangeVector(Minimum=<-250 150 0>, Maximum=<-250 150 200>))) {module=AddVelocity001}
   &AddVelocity(Velocity=&FN_FrontEndVector3Modifier(Gameplay=<0 0 0>, Front End=&RandomRangeVector(Minimum=<-10 -400 0>, Maximum=<10 -200 200>))) {enabled=false, module=AddVelocity002}
  ]
  particleUpdateScript [
   &ParticleState()
   &SkeletalMeshLocation(
    Apply To Particle Position=false
    Bone / Socket Index=2
    Bone Sampling Mode=::NewEnumerator4
    Inherit Velocity=false
    MaskSpawnGroup=false
    Sampled Position Offset=&Add_Vector(
     A=<-22.1231 -11 -0.8408>
     B=&RotateVectorByQuaternion(Vector To Rotate=&RandomRangeVector(Minimum=<0 16.1011 0>, Maximum=<30 0 2>), Quaternion=<0 -0.15 0 1>)
    )
    Sampled Position Offset Coordinate Space:ENiagaraMeshTransforms=::NewEnumerator3
    Skeletal Mesh=$User.SkeletalMesh
   ) {module=SkeletalMeshLocation001}
   /SetVariables(Particles.SourceDistance:NiagaraFloat=/Niagara/DynamicInputs/TypeConversions/VectorLengthSafe(Vector:Vector3f=&Subtract_Vector(A=$Output.SkeletalMeshLocation001.SampledPosition, B=Linked() {linkedVariable {name=Particles.Position, type=NiagaraPosition}}))) {isSetParameters=true}
   /Niagara/Modules/Update/Forces/CurlNoiseForce(
    Noise Frequency:NiagaraFloat=&RandomRangeFloat(Minimum=0.01, Maximum=0.25)
    Noise Strength=2000
    Pan Noise=true
    Pan Noise Field=<0 0 1>
   )
   /Niagara/Modules/Update/Forces/PointAttractionForce(
    Attraction Radius=1000
    AttractionStrength:NiagaraFloat=&Lerp_Float(A=0, B=1, Alpha=$Particles.SourceDistance)
    AttractorPosition:Vector3f=$Output.SkeletalMeshLocation001.SampledPosition
    Module.Kill Within Radius=false
    Module.Use Falloff=true
   ) {module=PointAttractionForce002}
   &Drag(Use Linear Drag=true, Use Rotational Drag=false)
   &DynamicMaterialParameters(
    Index 0 Param 1=&UniformRangedFloat(Minimum=4, Maximum=5)
    Index 0 Param 2=&FloatFromCurve(FloatCurve=DataInterface() {propertyValues {Curve {keys [
         {interpMode=RCIM_Cubic, tangentMode=RCTM_User, time=0, value=0}
         {arriveTangent=0.002, interpMode=RCIM_Cubic, leaveTangent=0.002, tangentMode=RCTM_User, time=0.1827, value=1}
         {arriveTangent=-1.2195, interpMode=RCIM_Cubic, leaveTangent=-1.2195, tangentMode=RCTM_User, time=0.9833, value=0.6448}
         <at 1.0011 0.002>
        ]}}})
    Index 1 Param 1=&UniformRangedFloat(Minimum=0, Maximum=30)
    Index 1 Param 2=1.1418
    Index 1 Param 3=&UniformRangedFloat(Minimum=0, Maximum=1)
    Index 1 Param 4=&UniformRangedFloat(Minimum=0, Maximum=1)
    Module.Param0WriteEnabled=false
    Module.Param0WriteEnabled1=true
    Module.Param1WriteEnabled=true
    Module.Param1WriteEnabled1=false
    Module.Param2WriteEnabled=false
    Module.Param2WriteEnabled1=true
    Module.Param3WriteEnabled=false
    Module.Param3WriteEnabled1=true
    Write Parameter Index 1=true
   ) {module=DynamicMaterialParameters001}
   &SolveForcesAndVelocity(Clamp Velocity=false, Limit Acceleration=false)
   &ScaleMeshSize(Scale Factor=&VectorFromFloat(Value=&FloatFromCurve(
      FloatCurve=DataInterface() {propertyValues {Curve {keys [
          {interpMode=RCIM_Cubic, tangentMode=RCTM_User, time=-0.0046, value=1.0113}
          {arriveTangent=-1.0067, interpMode=RCIM_Cubic, leaveTangent=-1.0067, tangentMode=RCTM_User, time=0.9283, value=0.7392}
          <at 1 0>
         ]}}}
      Scale Curve=2
     )))
  ]
  rendererClasses [&NiagaraMeshRendererProperties]
  renderers [{rendererClass=&NiagaraMeshRendererProperties, rendererIndex=0}]
 }
 Fly_Trail {
  emitterUpdateScript [
   &EmitterState(Enable Distance Culling=true, MaxDistance=2500, Module.bUseMaxDistance=true, Module.bUseMinDistance=false, Scalability Mode=::NewEnumerator1)
   /SetVariables(Emitter.EmitterReader=DataInterface() {propertyValues {EmitterBinding {bindingMode=Other, emitterName=Flies_Optimized}}}) {isSetParameters=true}
   &SpawnRate(
    SpawnRate=/Game/Effects/Niagara/DynamicInputs/Attributes/FN_GetEmitterParticleCount(AttributeReader:NiagaraDataInterfaceParticleRead=$Emitter.EmitterReader, Multiplier:NiagaraFloat=20)
    Use Spawn Probability=true
   )
  ]
  particleSpawnScript [
   &self:Sample_Location(New Particle Attribute Reader:NiagaraDataInterfaceParticleRead=$Emitter.EmitterReader) {module=Sample_Location}
   &InitializeRibbon(
    Color=<rgba 1 1 1 0.4>
    Lifetime=0.5
    Module.Write Color=true
    Module.Write Lifetime=true
    Module.Write Mass=true
    Module.Write Material Random=true
    Module.Write Position=false
    Module.Write RibbonTwist=false
    Module.Write RibbonWidth=true
    Ribbon Width=&RandomRangeFloat(Minimum=2, Maximum=6)
   )
   &AddVelocity(Velocity=&FN_FrontEndVector3Modifier(Gameplay=<0 0 0>, Front End=<-125 75 75>))
   &AddVelocity(Velocity=&FN_FrontEndVector3Modifier(Gameplay=<0 -200 0>, Front End=<0 0 75>)) {enabled=false, module=AddVelocity001}
  ]
  particleUpdateScript [
   &ParticleState()
   &SolveForcesAndVelocity(Clamp Velocity=false, Limit Acceleration=false)
   &ScaleColor(
    Scale Alpha=/Niagara/DynamicInputs/Math/LerpMultipleFloats(
     Float Count:ENiagara_MultipleLerpCount=::NewEnumerator2
     1:NiagaraFloat=0
     2:NiagaraFloat=1
     3:NiagaraFloat=1
     4:NiagaraFloat=0
     5:NiagaraFloat=0
     Alpha:NiagaraFloat=$Particles.NormalizedAge
    )
    ScaleA=true
    ScaleRGB=false
   )
   &ScaleRibbonWidth(Ribbon Width Scale=&FloatFromCurve(
     FloatCurve=DataInterface() {propertyValues {Curve {keys [<at 0 1 RCIM_Cubic>,<at 1 0>]}}}
     Scale Curve=&VectorLength(Vector=$Particles.InheritedScale)
    ))
  ]
  rendererClasses [&NiagaraRibbonRendererProperties]
  renderers [{rendererClass=&NiagaraRibbonRendererProperties}]
 }
 MeshTrails {
  emitterSpawnScript [&FN_PreserveSkydiveUserVariables() {module=PreserveUserVariables}]
  emitterUpdateScript [
   &EmitterState()
   &SpawnBurst_Instantaneous(Spawn Count=HlslExpression() {hlslExpression="Emitter.LoopCount == 0"}, Use Spawn Probability=false) {module=SpawnBurst_Instantaneous001}
  ]
  particleSpawnScript [
   &InitializeParticle(
    Color=<rgba 1 1 1 0.05>
    Module.Write Color=true
    Module.Write Lifetime=true
    Module.Write Mass=false
    Module.Write Material Random=true
    Module.Write Position=false
    Module.Write Scale=false
    Module.Write SpriteRotation=false
    Module.Write SpriteSize=false
   )
   &self:Mesh_Trail_Setup(Trail Length=250, Trail Width=75) {module=Mesh_Trail_Setup}
  ]
  particleUpdateScript [
   &ParticleState(Kill Particles When Lifetime Has Elapsed=false, Let Infinitely Lived Particles Die When Emitter Deactivates=true)
   &SkeletalMeshLocation(
    Apply Sampled Orientation to Mesh Particles=true
    Apply To Particle Position=false
    Bone / Socket Index=2
    Bone Sampling Mode=::NewEnumerator4
    Inherit Velocity=false
    Interpolate Mesh Sampling=false
    MaskSpawnGroup=false
    Sampled Position Offset Coordinate Space:ENiagaraMeshTransforms=::NewEnumerator3
    Skeletal Mesh=$User.SkeletalMesh
    Write Sampled Mesh Attrributes to Particle Attributes=true
   )
   /SetVariables(
    Particles.TrailOrient:Quat4f=&MultiplyQuaternion(
     Quaternion A=$Output.SkeletalMeshLocation.SampledRotation
     Quaternion B=/Niagara/DynamicInputs/TypeConversions/MakeCustomQuaternionFromBool(Bool:NiagaraBool=$User.bIsFrontEndPreview, True Quat:Quat4f=<0.21 -0.13 1 1>, False Quat:Quat4f=<0 0 1 1>)
    )
    Particles.TrailPosition:Vector3f=&Add_Vector(
     A=$Particles.SkeletalMeshLocation.SampledPosition
     B=&RotateVectorByQuaternion(Vector To Rotate=<137.1723 0 0>, Quaternion=$Particles.SkeletalMeshLocation.SampledRotation)
    )
    Particles.TrailTwinOrient:Quat4f=&MultiplyQuaternion(
     Quaternion A=$Output.SkeletalMeshLocation.SampledRotation
     Quaternion B=&MakeCustomQuaternionFromBool(Bool:NiagaraBool=$User.bIsFrontEndPreview, True Quat:Quat4f=<1 1 0.21 -0.13>, False Quat:Quat4f=<1 1 0 0>)
    )
    Particles.TrailTwinPosition:Vector3f=&Add_Vector(
     A=$Particles.SkeletalMeshLocation.SampledPosition
     B=&RotateVectorByQuaternion(Vector To Rotate=<-134.4295 0 0>, Quaternion=$Particles.SkeletalMeshLocation.SampledRotation)
    )
   ) {isSetParameters=true}
   &self:Exhaust_to_DynaMat_01(
    ExhaustVector:Vector3f=&FN_FrontEndVector3Modifier(
     Gameplay=&Multiply_Vector(A=$System.CurrentExhaust, B=<1 10 1>)
     Front End=&self:Front_End_Exhaust_01(
      Rate Of Change:NiagaraFloat=10
      Velocity Input:Vector3f=&TransformVector(Vector=$Particles.SkeletalMeshLocation.SampledVelocity, Source Space=::World, Destination Space=::Local)
      Velocity Multiplier:Vector3f=<0 -7 0>
     )
    )
    SpeedMult=1
   ) {module=Exhaust_to_DynaMat_01}
   &self:Mesh_Trail_Length_Update(Dynamic Length Scale:NiagaraFloat=&FN_FrontEndFloatModifier(Gameplay=$System.TrailLength, Front End=1)) {module=Mesh_Trail_Length_Update}
   &ScaleColor(
    Scale Alpha=&FN_FrontEndFloatModifier(Gameplay=$System.TrailOpacity, Front End=1)
    ScaleA=true
    ScaleRGB=false
   )
   &Color(Color=&ColorFromCurve(ColorCurve=DataInterface() {propertyValues {
       AlphaCurve {keys [<at 0 20 RCIM_Cubic>]}
       BlueCurve {keys [<at 0 0 RCIM_Cubic>,<at 0.5257 0>,<at 1 5 RCIM_Cubic>]}
       GreenCurve {keys [<at 0 0 RCIM_Cubic>,<at 0.5257 5>,<at 1 0 RCIM_Cubic>]}
       RedCurve {keys [<at 0 5 RCIM_Cubic>,<at 0.5257 0>,<at 1 0 RCIM_Cubic>]}
      }})) {enabled=false}
  ]
  rendererClasses [&NiagaraMeshRendererProperties]
  renderers [
   {rendererClass=&NiagaraMeshRendererProperties, rendererIndex=0}
   {rendererClass=&NiagaraMeshRendererProperties, rendererIndex=1}
  ]
 }
 NE_Empty {
  bEnabled=false
  emitterUpdateScript [
   &EmitterState(Life Cycle Mode=::NewEnumerator1, Loop Duration=1, UseLoopDelay=false)
   &SpawnRate(SpawnRate=5, Use Spawn Probability=false)
   /SetVariables(Emitter.AngleInput:NiagaraFloat=$User.AngleDot) {isSetParameters=true}
   /SetVariables(Emitter.CameraAngleInput:Vector3f=$User.CameraAngle) {isSetParameters=true}
  ]
  particleSpawnScript [
   &InitializeParticle(
    Lifetime=0.5
    Module.Write Color=true
    Module.Write Lifetime=true
    Module.Write Mass=true
    Module.Write Material Random=true
    Module.Write Position=false
    Module.Write Scale=false
    Module.Write SpriteRotation=false
    Module.Write SpriteSize=true
    Position=&SimulationPosition()
    Sprite Size=<75 75>
   )
   &SkeletalMeshLocation(
    Bone Sampling Mode=::NewEnumerator4
    Inherit Velocity=false
    MaskSpawnGroup=false
    Sampled Position Offset=<0 19.2311 91.4651>
    Sampled Position Offset Coordinate Space:ENiagaraMeshTransforms=::NewEnumerator3
    Skeletal Mesh=$User.SkeletalMesh
   ) {module=SkeletalMeshLocation001}
   /SetVariables(Particles.test:NiagaraFloat=$System.TimeOfDayCurve) {enabled=false, isSetParameters=true}
   &Color(Color=&ColorFromCurve(
     ColorCurve=DataInterface() {propertyValues {
       AlphaCurve {keys [<at 0 0.2 RCIM_Cubic>]}
       BlueCurve {keys [<at 0 0 RCIM_Cubic>,<at 0.3333 0>,<at 0.6667 0>,<at 1 1 RCIM_Cubic>]}
       GreenCurve {keys [<at 0 0 RCIM_Cubic>,<at 0.3333 1>,<at 0.6667 1>,<at 1 0 RCIM_Cubic>]}
       RedCurve {keys [<at 0 1 RCIM_Cubic>,<at 0.3333 1>,<at 0.6667 0>,<at 1 0 RCIM_Cubic>]}
      }}
     CurveIndex=&self:InterpretAngleDot(World Velocity:Vector3f=$Engine.Owner.Velocity)
    )) {enabled=false}
  ]
  particleUpdateScript [
   &ParticleState()
   &SolveForcesAndVelocity(Clamp Velocity=false, Limit Acceleration=false)
   &SkeletalMeshLocation(
    Bone Sampling Mode=::NewEnumerator4
    Inherit Velocity=false
    MaskSpawnGroup=false
    Sampled Position Offset=<0 19.2311 91.4651>
    Sampled Position Offset Coordinate Space:ENiagaraMeshTransforms=::NewEnumerator3
    Skeletal Mesh=$User.SkeletalMesh
   ) {module=SkeletalMeshLocation002}
   &Color(Color=&MakeLinearColorFromVectorAndFloat("Vector (RGB)"=$Particles.Velocity, "Float (Alpha)"=1)) {module=Color001}
  ]
  rendererClasses [&NiagaraSpriteRendererProperties]
  renderers [{rendererClass=&NiagaraSpriteRendererProperties}]
 }
 Smoke {
  emitterUpdateScript [
   &EmitterState()
   &SpawnRate(SpawnRate=80, Use Spawn Probability=false) {module=SpawnRate001}
  ]
  particleSpawnScript [
   &InitializeParticle(
    Lifetime=&RandomRangeFloat(Minimum=0.45, Maximum=0.75)
    Module.Write Color=true
    Module.Write Lifetime=true
    Module.Write Mass=true
    Module.Write Material Random=true
    Module.Write Position=false
    Module.Write Scale=false
    Module.Write SpriteRotation=true
    Module.Write SpriteSize=true
    Position=&SimulationPosition()
    Sprite Rotation=&UniformRangedFloat(Minimum=0, Maximum=360)
    Sprite Size=&Vector2DFromFloat(Value=&RandomRangeFloat(Minimum=17, Maximum=30))
   )
   &AccelerationForce(Acceleration=&UniformRangedVector(Minimum=<0 0 0>, Maximum=<0 0 50>))
   &ApplyInitialForces()
   /SetVariables(
    Particles.PlacementRandom:NiagaraFloat=&RandomRangeFloat(Minimum=0, Maximum=1)
    Particles.SpawnGroup=&SpawnGroup()
   ) {isSetParameters=true}
   /SetVariables(
    Particles.HeightOffset:Vector3f=&Add_Vector(
     A=<0 0 26.5627>
     B=&Multiply_VectorByFloat(
      Vector=<0 0 1>
      Float=&Multiply_Float(
       A=35
       B=&OneMinusFloat(Float=&Multiply_Float(A=$Particles.PlacementRandom, B=$Particles.PlacementRandom))
      )
     )
    )
    Particles.LengthOffset:Vector3f=&Add_Vector(
     A=<35 0 0>
     B=&Multiply_VectorByFloat(
      Vector=<1 0 0>
      Float=&Multiply_Float(
       A=$Particles.PlacementRandom
       B=&Multiply_Float(A=55, B=&UniformAOrBFloat(A=-1))
      )
     )
    )
    Particles.WidthOffset:Vector3f=&Multiply_VectorByFloat(
     Vector=<0 1 0>
     Float=&Multiply_Float(
      A=&OneMinusFloat(Float=$Particles.PlacementRandom)
      B=&Multiply_Float(A=15, B=&RandomRangeFloat(Minimum=-1, Maximum=1))
     )
    )
   ) {isSetParameters=true}
   &SkeletalMeshLocation(
    Bone / Socket Index=1
    Bone Sampling Mode=::NewEnumerator4
    Inherit Velocity=false
    MaskSpawnGroup=true
    Sampled Position Offset=&Add_Vector(A=$Particles.HeightOffset, B=&Add_Vector(A=$Particles.LengthOffset, B=$Particles.WidthOffset))
    Sampled Position Offset Coordinate Space:ENiagaraMeshTransforms=::NewEnumerator3
    Skeletal Mesh=$User.SkeletalMesh
   )
   &InheritVelocity(
    Module.Limit Speed=false
    Velocity Scale=&Multiply_VectorByFloat(Vector=<0 0 0.7>, Float=&Lerp_Float(A=1, B=0, Alpha=$Particles.NormalizedAge))
   )
  ]
  particleUpdateScript [
   &ParticleState()
   &SkeletalMeshLocation(
    Apply To Particle Position:NiagaraBool=&SetBoolByFloatComparison(A=$Particles.NormalizedAge, B=0.25, Comparison Type=::NewEnumerator6)
    Bone / Socket Index=1
    Bone Sampling Mode=::NewEnumerator4
    Inherit Velocity=false
    MaskSpawnGroup=false
    Sampled Position Offset=&Add_Vector(A=$Particles.HeightOffset, B=&Add_Vector(A=$Particles.LengthOffset, B=$Particles.WidthOffset))
    Sampled Position Offset Coordinate Space:ENiagaraMeshTransforms=::NewEnumerator3
    Skeletal Mesh=$User.SkeletalMesh
   ) {module=SkeletalMeshLocation002}
   &InheritVelocity(
    Module.Limit Speed=false
    Velocity Scale=&Multiply_VectorByFloat(Vector=<0.08 0.08 0>, Float=&Lerp_Float(A=1, B=0, Alpha=$Particles.NormalizedAge))
   ) {module=InheritVelocity001}
   &AddVelocity(Velocity=&FN_FrontEndVector3Modifier(Gameplay=<0 0 0>, Front End=&RandomRangeVector(Minimum=<-16 8 0.5>, Maximum=<-24 12 2>)))
   &AddVelocity(Velocity=&FN_FrontEndVector3Modifier(Gameplay=<0 0 0>, Front End=&RandomRangeVector(Minimum=<-3 -20 0.5>, Maximum=<3 -10 10>))) {enabled=false, module=AddVelocity001}
   &SolveForcesAndVelocity(Clamp Velocity=false, Limit Acceleration=false)
   &ScaleSpriteSize(Scale Factor=&Vector2DFromFloat(Value=&FloatFromCurve(
      FloatCurve=DataInterface() {propertyValues {Curve {keys [{arriveTangent=2.2383, interpMode=RCIM_Cubic, leaveTangent=2.2383, tangentMode=RCTM_User, time=0, value=0.25}, <at 1 1 RCIM_Cubic>]}}}
      Scale Curve=2
     )))
   /SetVariables(Particles.StartFrame:NiagaraFloat=&RandomRangeFloat(Minimum=0, Maximum=63)) {isSetParameters=true}
   /SetVariables(Particles.SubImageIndex=&Lerp_Float(
     A=$Particles.StartFrame
     B=&Add_Float(A=$Particles.StartFrame, B=&RandomRangeFloat(Minimum=4, Maximum=8))
     Alpha=$Particles.NormalizedAge
    )) {isSetParameters=true}
   &DynamicMaterialParameters(
    Index 0 Param 1=&FloatFromCurve(FloatCurve=DataInterface() {propertyValues {Curve {keys [
         {arriveTangent=4.0913, interpMode=RCIM_Cubic, leaveTangent=4.0913, tangentMode=RCTM_User, time=0, value=0}
         {arriveTangent=0.1513, interpMode=RCIM_Cubic, leaveTangent=0.1513, tangentMode=RCTM_User, time=0.5, value=1}
         {arriveTangent=-0.0082, interpMode=RCIM_Cubic, leaveTangent=-0.0082, tangentMode=RCTM_User, time=1, value=0}
        ]}}})
    Index 0 Param 4=1
    Module.Param0WriteEnabled=true
    Module.Param1WriteEnabled=false
    Module.Param2WriteEnabled=false
    Module.Param3WriteEnabled=true
   )
   &ScaleColor(
    Scale Alpha=&FloatFromCurve(
     FloatCurve=DataInterface() {propertyValues {Curve {keys [
         {arriveTangent=-2.2302, interpMode=RCIM_Cubic, leaveTangent=-2.2302, tangentMode=RCTM_User, time=0, value=1}
         {arriveTangent=-0.022, interpMode=RCIM_Cubic, leaveTangent=-0.022, tangentMode=RCTM_User, time=1, value=0}
        ]}}}
     Scale Curve=&LerpMultipleFloats(
      Float Count:ENiagara_MultipleLerpCount=::NewEnumerator0
      1:NiagaraFloat=&RandomRangeFloat(Minimum=0.02, Maximum=0.04)
      2:NiagaraFloat=&RandomRangeFloat(Minimum=0.05, Maximum=0.1)
      3:NiagaraFloat=&RandomRangeFloat(Minimum=0.02, Maximum=0.04)
      4:NiagaraFloat=?
      5:NiagaraFloat=?
      Alpha:NiagaraFloat=Linked() {linkedVariable {name=NPC.FortniteNPC.TimeOfDay_0-1, type=NiagaraFloat}}
     )
    )
    Scale RGB=<10 10 10>
    ScaleA=true
    ScaleRGB=false
   )
  ]
  rendererClasses [&NiagaraSpriteRendererProperties]
  renderers [
   {rendererClass=&NiagaraSpriteRendererProperties}
   {rendererClass=&NiagaraSpriteRendererProperties, rendererIndex=1}
  ]
 }
 VertCounter {
  bEnabled=false
  emitterUpdateScript [
   &EmitterState()
   &SpawnRate(SpawnRate=100, Use Spawn Probability=false)
  ]
  particleSpawnScript [
   /SetVariables(Particles.PickBone:NiagaraInt32=&RandomRangeInt(Minimum=421, Maximum=491)) {enabled=false, isSetParameters=true, module=SetVariables_#1}
   /SetVariables(Particles.PickMaxBone=1252, Particles.PickMinBone=1171) {isSetParameters=true}
   /SetVariables(
    Particles.PickAmount:NiagaraInt32=/Niagara/DynamicInputs/Subtract/Subtract_Integer(A:NiagaraInt32=$Particles.PickMaxBone, B:NiagaraInt32=$Particles.PickMinBone)
    Particles.PickBone:NiagaraInt32=&RandomRangeInt(Minimum=$Particles.PickMinBone, Maximum=$Particles.PickMaxBone)
   ) {isSetParameters=true}
   &InitializeParticle(
    Mesh Scale=<0.75 0.75 0.02>
    Module.Write Color=false
    Module.Write Lifetime=true
    Module.Write Mass=true
    Module.Write Material Random=true
    Module.Write Position=false
    Module.Write Scale=true
    Module.Write SpriteRotation=false
    Module.Write SpriteSize=true
    Sprite Size=&Vector2DFromFloat(Value=3)
   )
  ]
  particleUpdateScript [
   &ParticleState()
   &SkeletalMeshLocation(
    Apply To Particle Position=false
    Inherit Velocity=false
    MaskSpawnGroup=false
    Mesh Sampling Type=::NewEnumerator8
    OffsetAlongNormal=false
    Skeletal Mesh=$User.SkeletalMesh
    Vertex ID:NiagaraInt32=$Particles.PickBone
    Vertex Sampling Mode:ENiagaraVertexSamplingMode=::NewEnumerator8
   )
   &SkeletalMeshLocation(
    Apply To Particle Position=false
    FixedBarycentric=false
    Inherit Velocity=false
    MaskSpawnGroup=false
    Mesh Sampling Type=::NewEnumerator7
    OffsetAlongNormal=false
    Skeletal Mesh=$User.SkeletalMesh
    Triangle ID:NiagaraInt32=$Particles.PickBone
    Triangle Sampling Mode:ENiagaraTriangleSamplingMode=::NewEnumerator7
   ) {enabled=false, module=SkeletalMeshLocation001}
   &SkeletalMeshLocation(
    Apply To Particle Position=false
    Inherit Velocity=false
    MaskSpawnGroup=false
    Mesh Sampling Type=::NewEnumerator8
    OffsetAlongNormal=false
    Skeletal Mesh=$User.SkeletalMesh
    Vertex ID:NiagaraInt32=$Particles.PickBone
    Vertex Sampling Mode:ENiagaraVertexSamplingMode=::NewEnumerator7
   ) {enabled=false, module=SkeletalMeshLocation002}
   /SetVariables(
    Particles.MeshNormal:Vector3f=$Output.SkeletalMeshLocation.SampledNormal
    Particles.MeshPosition:Vector3f=$Output.SkeletalMeshLocation.SampledPosition
   ) {isSetParameters=true, module=SetVariables_#2}
   &Color(Color=&ColorFromCurve(
     ColorCurve=DataInterface() {propertyValues {
       AlphaCurve {keys [<at 0 100 RCIM_Cubic>]}
       BlueCurve {keys [<at 0 0 RCIM_Cubic>,<at 0.5257 0>,<at 1 1 RCIM_Cubic>]}
       GreenCurve {keys [<at 0 0 RCIM_Cubic>,<at 0.5257 1>,<at 1 0 RCIM_Cubic>]}
       RedCurve {keys [<at 0 1 RCIM_Cubic>,<at 0.5257 0>,<at 1 0 RCIM_Cubic>]}
      }}
     CurveIndex=&Divide_Float(
      A=&MakeFloatFromInt(Int=&Subtract_Integer(A:NiagaraInt32=$Particles.PickBone, B:NiagaraInt32=$Particles.PickMinBone))
      B=&MakeFloatFromInt(Int=$Particles.PickAmount)
     )
    ))
  ]
  rendererClasses [&NiagaraSpriteRendererProperties]
  renderers [{rendererClass=&NiagaraSpriteRendererProperties}]
 }
}
summary {
 systemName=NS_PieMan_Glider
 emitters [
  {emitterName=DroolPendulum, rendererClasses [&NiagaraSpriteRendererProperties, &NiagaraMeshRendererProperties]}
  {bEnabled=false, emitterName=VertCounter, rendererClasses [&NiagaraSpriteRendererProperties]}
  {emitterName=Smoke, rendererClasses [&NiagaraSpriteRendererProperties]}
  {emitterName=Flies_Optimized, rendererClasses [&NiagaraMeshRendererProperties]}
  {emitterName=Fly_Trail, rendererClasses [&NiagaraRibbonRendererProperties]}
  {bEnabled=false, emitterName=NE_Empty, rendererClasses [&NiagaraSpriteRendererProperties]}
  {emitterName=Drool, rendererClasses [&NiagaraSpriteRendererProperties, &NiagaraMeshRendererProperties]}
  {emitterName=MeshTrails, rendererClasses [&NiagaraMeshRendererProperties]}
 ]
 userVariables [
  User.AngleDot {type=NiagaraFloat, defaultValue=0, description=""}
  User.bIsFrontEnd {type=NiagaraBool}
  User.bIsFrontEndPreview {type=NiagaraBool}
  User.CameraAngle {type=Vector3f, description="", defaultValue {struct=Vector3f, value=<0 0 0>}}
  User.ForwardDot {type=NiagaraFloat, defaultValue=0, description=""}
  User.PlayerSpeed {type=NiagaraFloat, defaultValue=0, description=""}
  User.RightDot {type=NiagaraFloat, defaultValue=0, description=""}
  User.SkeletalMesh {
   type=NiagaraDataInterfaceSkeletalMesh
   defaultValue {
    struct=NiagaraExt_VariableValue_DataInterface
    value {dataInterface=&self:NiagaraDataInterfaceSkeletalMesh_0}
   }
  }
 ]
}
